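# Unlit, translucent, two-sided particle sprite material for bomb sparks in the Farmstead fort mission.
# Particle color is multiplied by the blue channel of the wall sparkle texture, boosted by a constant and fed into the shared FX emissive adjustments function.
# A separate multiply combines particle alpha with the texture's RGB, apparently as an opacity mask for fading.
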Material /Game/Missions/Primary/OB_FarmsteadFort/ArtAssets/Materials/M_BombSpark
graph {
 properties {
  bUsedWithParticleSprites=true
  blendMode=BLEND_Translucent
  pixelDepthOffsetMode=PDOM_Legacy
  shadingModel=MSM_Unlit
  twoSided=true
  usedShadingModels=Unlit
 }
 nodes [
  n1 = ParticleColor()
  n2 = TextureSample() {texture=/Game/Effects/Fort_Effects/Textures/T_Wall_Sparkle}
  MaterialFunctionCall(Emissive=Multiply(A=Multiply(A=@n1.RGB, B=@n2.B), B=Constant() {r=10})) {materialFunction=&MF_FX_EmissiveAdjustments}
  Multiply(A=@n1.A, B=@n2.RGB)
 ]
}
