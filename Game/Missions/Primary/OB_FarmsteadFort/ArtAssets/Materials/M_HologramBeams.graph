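# Unlit, two-sided translucent material for glowing hologram or scanner beams on mesh particles.
# Glow comes from a beam color parameter tinted by particle color.
# Opacity combines inverted Fresnel, vertex color and a ceiled beam texture, panned back and forth by two summed sines with a dynamic-parameter offset for variation.
# Particle and vertex alpha, depth fade and near-camera fade attenuate the result.
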Material /Game/Missions/Primary/OB_FarmsteadFort/ArtAssets/Materials/M_HologramBeams
graph {
 properties {
  bUsedWithMeshParticles=true
  bUsedWithNiagaraMeshParticles=true
  blendMode=BLEND_Translucent
  pixelDepthOffsetMode=PDOM_Legacy
  previewMesh=/Game/Gadgets/Assets/Bot_Turret/Meshes/Sm_BotTurret_ScannerRay_01
  shadingModel=MSM_Unlit
  twoSided=true
  usedShadingModels=Unlit
 }
 nodes [
  n1 = VectorParameter() {parameterName=ScanBeamColor, defaultValue=<rgba 10 10 10 1>}
  n2 = ParticleColor()
  MaterialFunctionCall(Emissive=Multiply(A=Multiply(A=@n1.RGB, B=@n2.RGB), B=ScalarParameter() {defaultValue=0.1, parameterName=BeamGlow})) {materialFunction=&MF_FX_EmissiveAdjustments}
  n3 = MaterialFunctionCall(
   Brightness=ScalarParameter() {defaultValue=1.3, parameterName=Fresnel Brightness}
   Exponent=ScalarParameter() {defaultValue=8, parameterName=Fresnel Exponent}
  ) {materialFunction=&MF_Fresnel_FX}
  n4 = VertexColor()
  n5 = Constant() {r=1}
  n6 = ScalarParameter() {defaultValue=0.01, parameterName=Grid Brightness}
  n7 = TextureCoordinate() {uTiling=8}
  n8 = DynamicParameter() {defaultValue=<rgba 0 0 0 1>, paramNames [ScannerRandomPosition, Speed Random, Param3, Team Color Picker]}
  n9 = AppendVector(A=@n8.ScannerRandomPosition, B=Constant())
  n10 = Multiply(A=Time(), B=ScalarParameter() {defaultValue=13, parameterName=Speed A})
  n11 = Sine(None=@n10) {period=1.3}
  n12 = Sine(None=@n10) {period=1.7}
  n13 = TextureSample!(UVs=Panner(Coordinate=Add(A=@n7, B=@n9), Time=Add(A=@n11, B=@n12)) {speedX=2.6}, Tex, Level, Apply View MipBias) {
   constMipValue=0
   mipValueMode=TMVM_MipLevel
   samplerType=SAMPLERTYPE_LinearColor
   texture=/Game/Effects/Fort_Effects/Textures/Tile/T_Raygun_BeamReveal_01
  }
  n14 = Add(A=Add(A=Add(A=@n3.Fresnel Inverted, B=@n4.R), B=Multiply(A=@n5, B=@n6)), B=Ceil(None=@n13.R))
  n15 = Multiply(
   A=LinearInterpolate(
    A=ScalarParameter() {defaultValue=0.075, parameterName=Fresnel Inner}
    B=ScalarParameter() {defaultValue=1, parameterName=Fresnel Outer}
    Alpha=Clamp(None=@n14)
   )
   B=VertexColor().A
  )
  Multiply(
   A=DepthFade(Opacity=Multiply(A=@n2.A, B=@n15), FadeDistance=ScalarParameter() {defaultValue=35, parameterName=DepthFadeDistance})
   B=MaterialFunctionCall(Distance=ScalarParameter() {defaultValue=120, parameterName=NearCameraFadeDistance}) {materialFunction=&Near_Camera_Fade}.Result
  )
 ]
}
