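# Unlit translucent particle sprite material, named as a hologram core for the Farmstead Fort mission.
# Particle color RGB is boosted, passed through day/night auto bloom control, scaled up again, then fed to the FX emissive adjustments function.
# A starburst flare texture multiplied by particle alpha forms a separate branch, apparently for opacity, though the final material output links are not shown.
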Material /Game/Missions/Primary/OB_FarmsteadFort/ArtAssets/Materials/M_HologramCore
graph {
 properties {
  bUsedWithParticleSprites=true
  blendMode=BLEND_Translucent
  pixelDepthOffsetMode=PDOM_Legacy
  shadingModel=MSM_Unlit
  usedShadingModels=Unlit
 }
 nodes [
  n1 = Constant() {r=1}
  n2 = ParticleColor()
  n3 = Multiply(A=@n2.RGB) {constA=1, constB=2}
  MaterialFunctionCall(Emissive=Multiply(
    A=MaterialFunctionCall!(DayBloom_Mult=@n1, In=@n3) {materialFunction=/Game/Packages/Fortnite_UniqueMaterials_1/MH_Materials/DayNight_Auto_Bloom_Control}.Result
    B=Constant() {r=14}
   )) {materialFunction=&MF_FX_EmissiveAdjustments}
  Multiply(A=TextureSample() {texture=/Game/Effects/Fort_Effects/Textures/Flares/T_Starburst}.RGB, B=@n2.A)
 ]
}
