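# Default Lit, Nanite-enabled material for the Homebase blast door in the Farmstead Fort primary mission, with a scalable emissive map.
# The emissive texture is multiplied by the VDoor scalar parameter, so the glow strength can be varied.
# A mask texture's green and blue channels are scaled by constants, alongside diffuse and normal samples.
# The graph shows no links from these nodes to the material outputs.
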Material /Game/Missions/Primary/OB_FarmsteadFort/ArtAssets/Materials/M_Homebase_Blast_Door_Emissive
graph {
 properties {bUsedWithNanite=true, pixelDepthOffsetMode=PDOM_Legacy, usedShadingModels=Default Lit}
 nodes [
  Multiply(
   A=TextureSample() {texture=/Game/Missions/Primary/OB_FarmsteadFort/ArtAssets/Textures/Homebase_Blast_Door_E}.RGB
   B=ScalarParameter() {parameterName=VDoor}
  )
  n1 = TextureSample() {samplerType=SAMPLERTYPE_Masks, texture=/Game/Missions/Primary/OB_FarmsteadFort/ArtAssets/Textures/Homebase_Blast_Door_S}
  Multiply(A=@n1.G, B=Constant() {r=1})
  Multiply(A=@n1.B, B=Constant() {r=2})
  TextureSample() {texture=/Game/Missions/Primary/OB_FarmsteadFort/ArtAssets/Textures/Homebase_Blast_Door_D}
  TextureSample() {samplerType=SAMPLERTYPE_Normal, texture=/Game/Missions/Primary/OB_FarmsteadFort/ArtAssets/Textures/Homebase_Blast_Door_N}
 ]
}
