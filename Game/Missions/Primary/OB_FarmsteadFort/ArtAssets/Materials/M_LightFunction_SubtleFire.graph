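# Light function material that gives a light subtle, fire-like flicker, stored with the Farmstead Fort mission art.
# It blends two horizontally panning textures, a slow fire beam clip with a slight mix of a faster, tiled wave, and clamps the result to keep a brightness floor.
# That pattern is multiplied by two nested-lerp pulses built from time-driven sines, which share the default period and differ only in strength.
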Material /Game/Missions/Primary/OB_FarmsteadFort/ArtAssets/Materials/M_LightFunction_SubtleFire
graph {
 properties {materialDomain=MD_LightFunction, pixelDepthOffsetMode=PDOM_Legacy, usedShadingModels=Default Lit}
 nodes [
  n1 = TextureCoordinate()
  n2 = Time()
  n3 = TextureCoordinate() {uTiling=2, vTiling=2}
  n4 = Time()
  n5 = LinearInterpolate(
   A=TextureSample(UVs=Panner(Coordinate=@n1, Time=@n2) {speedX=0.04}) {texture=/Game/Weapons/FORT_Rifles/Effects/Textures/T_FireBeamClip}.RGB
   B=TextureSample(UVs=Panner(Coordinate=@n3, Time=@n4) {speedX=0.2}) {texture=/Game/Effects/Fort_Effects/Textures/waveH}.RGB
  ) {constAlpha=0.1}
  n6 = Time()
  n7 = LinearInterpolate(A=Constant() {r=1}, B=Constant() {r=0.7}, Alpha=Divide(A=Sine(None=@n6)) {constB=5})
  n8 = Time()
  n9 = LinearInterpolate(A=Constant() {r=1}, B=Constant() {r=0.9}, Alpha=Divide(A=Sine(None=@n8)) {constB=6})
  n10 = Divide(A=Sine(None=Time())) {constB=6}
  n11 = Constant() {r=1}
  n12 = Constant() {r=0.6}
  n13 = Divide(A=Sine(None=Time())) {constB=8}
  n14 = Constant() {r=1}
  n15 = Constant() {r=0.8}
  n16 = Divide(A=Sine(None=Time())) {constB=8}
  n17 = Sine(None=Time())
  Multiply(
   A=Multiply(A=Clamp(None=@n5) {minDefault=0.4}, B=LinearInterpolate(A=@n7, B=@n9, Alpha=@n10))
   B=LinearInterpolate(A=LinearInterpolate(A=@n11, B=@n12, Alpha=@n13), B=LinearInterpolate(A=@n14, B=@n15, Alpha=@n16), Alpha=Divide(A=@n17) {constB=5})
  )
 ]
}
