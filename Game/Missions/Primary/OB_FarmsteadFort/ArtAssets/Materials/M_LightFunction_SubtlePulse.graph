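# Light function material for the Farmstead Fort mission that gives a light a subtle flicker and pulse.
# Time slowly moves the sample point on a muzzle cloud texture, so light varies over time rather than projecting a spatial pattern.
# The doubled RGB is multiplied by a faster time-driven sine, clamped to a high floor so the pulse only briefly swells and never drives the light dark by itself.
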Material /Game/Missions/Primary/OB_FarmsteadFort/ArtAssets/Materials/M_LightFunction_SubtlePulse
graph {
 properties {materialDomain=MD_LightFunction, pixelDepthOffsetMode=PDOM_Legacy, usedShadingModels=Default Lit}
 nodes [
  n1 = Divide(A=Time(), B=Constant() {r=6})
  n2 = Divide(A=Time(), B=Constant() {r=1.5})
  Multiply(
   A=Multiply(A=TextureSample(UVs=@n1) {texture=/Game/Weapons/FORT_Pistols/Effects/Muzzle/T_Cloud_d}.RGB, B=Constant() {r=2})
   B=Clamp(None=Sine(None=@n2), Min=Constant() {r=0.65}, Max=Constant() {r=1})
  )
 ]
}
