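# Unlit additive material for the projector station cone in the FarmsteadFort mission, flagged for mesh and Niagara mesh particles.
# A panning radial gradient texture, tinted by a boosted Color parameter (blue by default), feeds the emissive adjustments function.
# A fade mask multiplies particle alpha, Fresnel and view-angle fade by a lengthwise V mask.
# A scalar parameter blends that V mask between a powered linear gradient and a clamped sine curve.
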Material /Game/Missions/Primary/OB_FarmsteadFort/ArtAssets/Materials/M_ProjectorStationCone
graph {
 properties {
  bEnableResponsiveAA=true
  bUsedWithMeshParticles=true
  bUsedWithNiagaraMeshParticles=true
  blendMode=BLEND_Additive
  pixelDepthOffsetMode=PDOM_Legacy
  shadingModel=MSM_Unlit
  usedShadingModels=Unlit
 }
 nodes [
  n1 = Panner(
   Coordinate=TextureCoordinate()
   Time=Time()
   Speed=MaterialFunctionCall(
    X=ScalarParameter() {defaultValue=1, parameterName=Pan_U, sliderMax=1}
    Y=ScalarParameter() {defaultValue=1, parameterName=Pan_V, sliderMax=1}
   ) {materialFunction=&MakeFloat2}.Result
  ) {speedX=1, speedY=1}
  n2 = Constant() {r=25}
  n3 = VectorParameter() {parameterName=Color, defaultValue=<rgba 0.2 0.4 1 0>}
  MaterialFunctionCall(Emissive=Multiply(A=TextureSample(UVs=@n1) {texture=/Game/Maps/Textures/RadialGradient2}.RGB, B=Multiply(A=@n2, B=@n3.RGB))) {materialFunction=&MF_FX_EmissiveAdjustments}
  n4 = Fresnel(ExponentIn=Constant() {r=1.5})
  n5 = MaterialFunctionCall("FadeSensitivity (optional)"=Constant() {r=4}) {materialFunction=&FadeBasedOnViewAngle}
  n6 = Power(Base=MaterialFunctionCall() {materialFunction=&LinearGradient}.VGradient) {constExponent=3}
  n7 = Divide(A=ComponentMask(None=TextureCoordinate()) {g=true})
  n8 = Clamp(None=Power(Base=Sine(None=@n7)))
  n9 = ScalarParameter() {parameterName=1ForSineCurveMask, sliderMax=1}
  Multiply(A=ParticleColor().A, B=Multiply(A=Multiply(A=@n4, B=@n5.Result), B=LinearInterpolate(A=@n6, B=@n8, Alpha=@n9)))
 ]
}
