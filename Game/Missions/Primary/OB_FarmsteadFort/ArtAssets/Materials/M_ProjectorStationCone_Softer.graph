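# Additive unlit mesh-particle material, a softer variant of the projector station light cone in the FarmsteadFort mission.
# A panning gradient texture plus a sharp V gradient is tinted by particle color; IsolateParticleColor moves that pattern from emissive into opacity.
# Opacity multiplies Fresnel, view-angle fade and a V gradient, softened by depth fade on the default shading path only.
# Static switches choose the Fresnel method and add center fade or particle alpha.
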Material /Game/Missions/Primary/OB_FarmsteadFort/ArtAssets/Materials/M_ProjectorStationCone_Softer
graph {
 properties {
  bEnableResponsiveAA=true
  bUsedWithMeshParticles=true
  bUsedWithNiagaraMeshParticles=true
  blendMode=BLEND_Additive
  pixelDepthOffsetMode=PDOM_Legacy
  shadingModel=MSM_Unlit
  usedShadingModels=Unlit
 }
 nodes [
  n1 = ParticleColor()
  n2 = Multiply(
   A=TextureCoordinate()
   B=AppendVector(A=ScalarParameter() {defaultValue=0.5, parameterName=Tiling-X}, B=ScalarParameter() {defaultValue=1, parameterName=Tiling-Y})
  )
  n3 = AppendVector(A=ScalarParameter() {defaultValue=0.1, parameterName=PanSpeed-X}, B=ScalarParameter() {parameterName=PanSpeed-Y})
  n4 = MaterialFunctionCall() {materialFunction=&LinearGradient}
  n5 = Add(
   A=TextureSampleParameter2D(UVs=Panner(Coordinate=@n2, Speed=@n3) {speedX=0.1}) {parameterName=Texture, texture=/Game/Maps/Textures/RadialGradient2}.RGB
   B=Power(Base=@n4.VGradient) {constExponent=12}
  )
  MaterialFunctionCall(Emissive=StaticSwitchParameter(True=@n1.RGB, False=Multiply(A=@n5, B=@n1.RGB)) {parameterName=IsolateParticleColor?}) {materialFunction=&MF_FX_EmissiveAdjustments}
  n6 = OneMinus(None=MaterialFunctionCall() {materialFunction=&LinearGradient}.VGradient)
  n7 = ScalarParameter() {defaultValue=5, parameterName=CenterFadeExp}
  n8 = MaterialFunctionCall(
   Brightness=ScalarParameter() {defaultValue=1, parameterName=FresnelBrightness}
   Exponent=ScalarParameter() {defaultValue=2, parameterName=FresnelExponent}
  ) {materialFunction=&MF_Fresnel_FX}
  n9 = Fresnel(ExponentIn=ScalarParameter() {defaultValue=2, parameterName=FresnelExponent})
  n10 = ScalarParameter() {defaultValue=4, parameterName=AngleFade}
  n11 = Multiply(
   A=Multiply(
    A=StaticSwitchParameter(True=@n8.Fresnel, False=@n9) {parameterName=UseFX_Fresnel}
    B=MaterialFunctionCall("FadeSensitivity (optional)"=@n10) {materialFunction=&FadeBasedOnViewAngle}.Result
   )
   B=Power(Base=@n4.VGradient, Exp=ScalarParameter() {defaultValue=4, parameterName=LinearGradient_Power}) {constExponent=4}
  )
  n12 = ShadingPathSwitch(
   Default=DepthFade(Opacity=@n11, FadeDistance=ScalarParameter() {defaultValue=70, parameterName=DepthFade}) {fadeDistanceDefault=70}
   Mobile=@n11
  ) {desc=Migrated from FeatureLevelSwitch}
  n13 = StaticSwitchParameter(True=Multiply(A=Power(Base=@n6, Exp=@n7) {constExponent=4}, B=@n12), False=@n12) {parameterName=UseCenterFade}
  n14 = StaticSwitchParameter(True=Multiply(A=@n5, B=@n13), False=@n13) {parameterName=IsolateParticleColor?}
  StaticSwitchParameter(True=Multiply(A=@n14, B=@n1.A), False=@n14) {parameterName=UseParticleAlpha?}
 ]
}
