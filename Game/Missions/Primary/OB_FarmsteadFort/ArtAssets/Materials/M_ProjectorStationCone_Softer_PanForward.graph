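# Additive, unlit, two-sided mesh-particle material for a soft projector station light cone with forward-panning detail.
# A vertical gradient, inverted Fresnel and panning, rotatable radial-gradient texture form the cone mask, depth-faded except on mobile, with optional particle-alpha scaling.
# Three fast-panning noise layers are summed, sharpened by a power, added to the powered mask, tinted by particle color and passed through the shared FX emissive-adjustment function.
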Material /Game/Missions/Primary/OB_FarmsteadFort/ArtAssets/Materials/M_ProjectorStationCone_Softer_PanForward
graph {
 properties {
  bEnableResponsiveAA=true
  bUsedWithMeshParticles=true
  bUsedWithNiagaraMeshParticles=true
  blendMode=BLEND_Additive
  pixelDepthOffsetMode=PDOM_Legacy
  shadingModel=MSM_Unlit
  twoSided=true
  usedShadingModels=Unlit
 }
 nodes [
  n1 = MaterialFunctionCall() {materialFunction=&LinearGradient}
  n2 = Power(Base=OneMinus(None=Fresnel() {exponent=0.1}))
  n3 = Panner(Coordinate=TextureCoordinate() {vTiling=0.25}) {speedY=0.25}
  n4 = ScalarParameter() {defaultValue=0.25, parameterName="Rotation Angle (0-1) (S)"}
  n5 = Multiply(
   A=Multiply(A=Power(Base=@n1.VGradient) {constExponent=1}, B=Clamp(None=@n2))
   B=TextureSample(UVs=MaterialFunctionCall(UVs=@n3, "Rotation Angle (0-1)"=@n4) {materialFunction=&CustomRotator}.Rotated Values) {texture=/Game/Maps/Textures/RadialGradient2}.G
  )
  n6 = ShadingPathSwitch(
   Default=DepthFade(Opacity=@n5, FadeDistance=ScalarParameter() {defaultValue=70, parameterName=Deapth Fade}) {fadeDistanceDefault=70}
   Mobile=@n5
  ) {desc=Migrated from FeatureLevelSwitch}
  n7 = ParticleColor()
  LinearInterpolate(A=@n6, B=Multiply(A=@n7.A, B=@n6), Alpha=ScalarParameter() {parameterName=Multiply by Particle Alpha})
  n8 = TextureCoordinate()
  n9 = ComponentMask(None=VectorParameter() {parameterName=Bubble_Tile_Scale_1, defaultValue=<rgba 1 1 0 0>}.RGB) {g=true, r=true}
  n10 = TextureSample(UVs=Panner(Coordinate=Multiply(A=@n8, B=@n9)) {speedX=-0.05, speedY=2.5}) {texture=/Game/Global/Textures/Default/Noise/T_Tiling_Noise_01}
  n11 = Add(A=TextureCoordinate() {vTiling=0.5}, B=VectorParameter() {parameterName=Bubble_UV_Scale_3, defaultValue=<rgba 0 0.25 0 0>}.G)
  n12 = ComponentMask(None=VectorParameter() {parameterName=Bubble_Tile_Scale_2, defaultValue=<rgba 1 1 0 0>}.RGB) {g=true, r=true}
  n13 = TextureSample(UVs=Panner(Coordinate=Multiply(A=@n11, B=@n12)) {speedX=0.01, speedY=4}) {texture=&T_Tiling_Noise_01}
  n14 = TextureCoordinate() {vTiling=0.4}
  n15 = VectorParameter() {parameterName=Bubble_UV_Scale_4, defaultValue=<rgba 0.5 0.25 0 0>}
  n16 = VectorParameter() {parameterName=Bubble_Tile_Scale_3, defaultValue=<rgba 1 1 0 0>}
  n17 = Panner(Coordinate=Multiply(A=Add(A=@n14, B=@n15.G), B=ComponentMask(None=@n16.RGB) {g=true, r=true})) {speedX=0.02, speedY=3}
  n18 = Power(
   Base=Add(A=Add(A=@n10.RGB, B=@n13.RGB), B=TextureSample(UVs=@n17) {texture=&T_Tiling_Noise_01}.RGB)
   Exp=ScalarParameter() {defaultValue=2, parameterName=BubblePowerMult}
  ) {constExponent=1}
  n19 = Power(Base=@n6)
  MaterialFunctionCall(Emissive=Multiply(A=Add(A=@n18, B=@n19), B=@n7.RGB)) {materialFunction=&MF_FX_EmissiveAdjustments}
 ]
}
