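# Translucent, lit material for a floor light prop in the FarmsteadFort mission, built on the siren light textures.
# One branch scales the siren light's _E texture by a scalar parameter, and another passes the blue channel of its _S mask texture through a neutral multiply.
# A Transparency scalar and a yellow color parameter sit alongside, using per-vertex non-directional volumetric translucency lighting.
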Material /Game/Missions/Primary/OB_FarmsteadFort/ArtAssets/Materials/M_Prop_FloorLight
graph {
 properties {
  blendMode=BLEND_Translucent
  pixelDepthOffsetMode=PDOM_Legacy
  translucencyLightingMode=TLM_VolumetricPerVertexNonDirectional
  usedShadingModels=Default Lit
 }
 nodes [
  Multiply(
   A=TextureSample() {texture=/Game/Missions/Primary/OB_FarmsteadFort/ArtAssets/Textures/T_Prop_Siren_Light_E}.RGB
   B=ScalarParameter() {defaultValue=0.23}
  )
  Multiply(
   A=TextureSample() {samplerType=SAMPLERTYPE_Masks, texture=/Game/Missions/Primary/OB_FarmsteadFort/ArtAssets/Textures/T_Prop_Siren_Light_S}.B
   B=Constant() {r=1}
  )
  ScalarParameter() {defaultValue=0.4, parameterName=Transparency}
  VectorParameter() {defaultValue=<rgba 0.945 0.8721 0.1661 1>}
 ]
}
